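# Material function that adds animated emissive flame and eye-flame effects to the Patches backpack's material attributes.
# Panning noise distorts flame-mask UVs, partly bump-offset for depth, to drive curve-atlas color ramps for body and eye flames.
# Mask textures composite these with tinted emissive regions under separate intensity controls.
# It also darkens masked base color and blends an interior normal map over the incoming normal.
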MaterialFunction /BRCosmetics/Accessories/FORT_Backpacks/Backpack_M_MED_Patches/Materials/MF_Patches_flame
graph {nodes [
  n1 = GetMaterialAttributes(None=FunctionInput() {inputType=FunctionInput_MaterialAttributes}) {attributeGetTypes [#1, #2]}
  n2 = TextureSampleParameter2D() {
   group=patches_material
   parameterName=FlameEmissiveMask
   samplerSource=SSM_Wrap_WorldGroupSettings
   samplerType=SAMPLERTYPE_Masks
   texture=/BRCosmetics/Global/Textures/Default/Blanks/AllGreen
  }
  n3 = Multiply(A=@n1.BaseColor, B=LinearInterpolate(A=Constant() {r=0.1}, Alpha=@n2.B)) {constB=0.1}
  n4 = TextureSampleParameter2D() {
   group=patches_material
   parameterName=FX_Mask
   samplerSource=SSM_Wrap_WorldGroupSettings
   samplerType=SAMPLERTYPE_Masks
   texture=/BRCosmetics/Global/Textures/Default/Blanks/AllGB
  }
  n5 = NamedRerouteUsage()
  n6 = TextureSample() {samplerType=SAMPLERTYPE_Normal, texture=/BRCosmetics/Accessories/FORT_Backpacks/Backpack_M_MED_Patches/Texture/T_M_MED_Patches_inside_Normal}
  FunctionOutput(None=SetMaterialAttributes!(
    MaterialAttributes=@n1.MaterialAttributes
    EmissiveColor=NamedRerouteUsage()
    BaseColor=LinearInterpolate(A=@n1.BaseColor, B=@n3, Alpha=@n4.G) {constB=0}
    Normal=MaterialFunctionCall(BaseNormal=@n5, AdditionalNormal=@n6.RGB) {materialFunction=&BlendAngleCorrectedNormals}.Result
   ) {attributeSetTypes [#3, #1, #2]})
  n7 = ScalarParameter() {defaultValue=1, group=patches_material, parameterName=EYE_EmissveIndensity}
  n8 = ScalarParameter() {defaultValue=17, parameterName=EYE_InsideIndensity}
  n9 = TextureSampleParameter2D() {
   group=patches_material
   parameterName=FlameEmissiveMask
   samplerSource=SSM_Wrap_WorldGroupSettings
   samplerType=SAMPLERTYPE_Masks
   texture=&AllGreen
  }
  n10 = Power(Base=@n9.A, Exp=Constant() {r=3})
  n11 = ScalarParameter() {defaultValue=10, group=patches_material, parameterName=Inside_EmissveIndensity}
  n12 = TextureSampleParameter2D() {
   group=patches_material
   parameterName=FlameEmissiveMask
   samplerSource=SSM_Wrap_WorldGroupSettings
   samplerType=SAMPLERTYPE_Masks
   texture=&AllGreen
  }
  n13 = Power(Base=NamedRerouteUsage()) {constExponent=3}
  n14 = Multiply(A=NamedRerouteUsage()) {constB=2}
  n15 = Multiply(A=Multiply(A=Saturate(None=@n14), B=NamedRerouteUsage()), B=Constant() {r=0.3}) {constB=0.5}
  n16 = CurveAtlasRowParameter(CurveTime=Add(A=Add(A=@n13, B=@n15), B=NamedRerouteUsage())) {
   atlas=&/patch_ColorAtlas
   curve=&/CC_patch
   group=patches_material
   parameterName=Curve
  }
  n17 = TextureSampleParameter2D() {
   group=patches_material
   parameterName=FlameEmissiveMask
   samplerSource=SSM_Wrap_WorldGroupSettings
   samplerType=SAMPLERTYPE_Masks
   texture=&AllGreen
  }
  n18 = Multiply(A=@n16, B=LinearInterpolate(A=Constant() {r=0.2}, B=Constant() {r=1.5}, Alpha=@n17.A))
  n19 = ScalarParameter() {defaultValue=2, group=patches_material, parameterName=global_EmissveIndensity}
  n20 = Multiply(
   A=LinearInterpolate(B=LinearInterpolate(A=@n7, B=@n8, Alpha=@n10), Alpha=@n9.B) {constA=1, constB=10}
   B=Multiply(A=LinearInterpolate(B=@n11, Alpha=@n12.R) {constA=1}, B=Multiply(A=@n18, B=@n19) {constB=2})
  )
  n21 = TextureSampleParameter2D() {
   group=patches_material
   parameterName=Emissive_color
   samplerSource=SSM_Wrap_WorldGroupSettings
   texture=/Game/Global/Textures/Default/Blanks/T_Blank_LegalWhite_sRGBColor
  }
  n22 = TextureSampleParameter2D() {
   group=patches_material
   parameterName=FlameEmissiveMask
   samplerSource=SSM_Wrap_WorldGroupSettings
   samplerType=SAMPLERTYPE_Masks
   texture=&AllGreen
  }
  n23 = TextureSampleParameter2D() {
   group=patches_material
   parameterName=FlameEmissiveMask
   samplerSource=SSM_Wrap_WorldGroupSettings
   samplerType=SAMPLERTYPE_Masks
   texture=&AllGreen
  }
  n24 = LinearInterpolate(
   A=LinearInterpolate(A=@n20, B=LinearInterpolate(A=@n20, B=@n21.RGB, Alpha=@n22.B), Alpha=Constant() {r=0.4}) {constAlpha=1}
   B=NamedRerouteUsage()
   Alpha=@n23.G
  )
  n25 = TextureSampleParameter2D() {
   group=patches_material
   parameterName=FX_Mask
   samplerSource=SSM_Wrap_WorldGroupSettings
   samplerType=SAMPLERTYPE_Masks
   texture=&AllGB
  }
  n26 = Constant3Vector() {constant=<rgba 0.5271 0 0.6376 1>}
  n27 = Constant() {r=0.3}
  n28 = TextureSampleParameter2D() {
   group=patches_material
   parameterName=Flame_mask
   samplerSource=SSM_Wrap_WorldGroupSettings
   samplerType=SAMPLERTYPE_Masks
   texture=/Game/Global/Textures/Default/Blanks/AllRed
  }
  n29 = Constant() {r=6}
  n30 = TextureSampleParameter2D() {
   group=patches_material
   parameterName=Emissive_color
   samplerSource=SSM_Wrap_WorldGroupSettings
   texture=&T_Blank_LegalWhite_sRGBColor
  }
  n31 = TextureSampleParameter2D() {
   group=patches_material
   parameterName=FlameEmissiveMask
   samplerSource=SSM_Wrap_WorldGroupSettings
   samplerType=SAMPLERTYPE_Masks
   texture=&AllGreen
  }
  n32 = TextureSampleParameter2D() {
   group=patches_material
   parameterName=Flame_mask
   samplerSource=SSM_Wrap_WorldGroupSettings
   samplerType=SAMPLERTYPE_Masks
   texture=&AllRed
  }
  n33 = LinearInterpolate(
   A=LinearInterpolate(A=Multiply(A=@n24, B=@n25.G), B=Multiply(A=@n26, B=@n27), Alpha=Power(Base=@n28.G, Exp=@n29))
   B=Multiply(A=@n30.RGB, B=LinearInterpolate(Alpha=@n31.A) {constA=5, constB=20}) {constB=10}
   Alpha=@n32.B
  )
  MaterialFunctionCall(Emissive=@n33) {materialFunction=&MF_FX_EmissiveAdjustments}
  n34 = TextureCoordinate()
  n35 = BumpOffset(Coordinate=@n34, Height=Constant() {r=-1.5})
  n36 = TextureSampleParameter2D() {
   group=patches_material
   parameterName=Flame_mask
   samplerSource=SSM_Wrap_WorldGroupSettings
   samplerType=SAMPLERTYPE_Masks
   texture=&AllRed
  }
  n37 = NamedRerouteUsage()
  n38 = Constant() {r=0.03}
  n39 = TextureSampleParameter2D(UVs=Add(A=LinearInterpolate(A=@n35, B=@n34, Alpha=@n36.B) {constB=0}, B=Multiply(A=@n37, B=@n38) {constB=0.1})) {
   group=patches_material
   parameterName=Flame_mask
   samplerSource=SSM_Wrap_WorldGroupSettings
   samplerType=SAMPLERTYPE_Masks
   texture=&AllRed
  }
  NamedRerouteDeclaration(None=@n39.R) {name=Flame mask, nodeColor=<rgba 0.2188 0 1 1>}
  n40 = TextureCoordinate()
  n41 = Constant() {r=3}
  n42 = Constant()
  n43 = TextureSampleParameter2D() {
   group=patches_material
   parameterName=FlameEmissiveMask
   samplerSource=SSM_Wrap_WorldGroupSettings
   samplerType=SAMPLERTYPE_Masks
   texture=&AllGreen
  }
  n44 = LinearInterpolate(A=Constant() {r=-0.5}, Alpha=@n43.B) {constB=0.5}
  n45 = TextureSampleParameter2D(UVs=Panner(Coordinate=Multiply(A=@n40, B=@n41), Speed=AppendVector(A=@n42, B=@n44)) {speedY=-0.5}) {
   group=patches_material
   parameterName=Distortion Texture
   samplerSource=SSM_Wrap_WorldGroupSettings
   samplerType=SAMPLERTYPE_Grayscale
   texture=/Game/TimeOfDay/TODM_Sequence/Materials/Sky/Textures/LowResBlurredNoise_Radish_LF_Root
  }
  NamedRerouteDeclaration(None=LinearInterpolate(A=Constant() {r=-0.5}, Alpha=@n45.R) {constA=-0.2}) {name=UV Distortion, nodeColor=<rgba 1 0.2109 0 1>}
  NamedRerouteDeclaration(None=@n1.Normal) {name=Normal, nodeColor=<rgba 1 0 0.1875 1>}
  NamedRerouteDeclaration(None=@n33) {name=Emissive, nodeColor=<rgba 0.8984 1 0 1>}
  n46 = TextureCoordinate()
  n47 = TextureSampleParameter2D() {
   group=patches_material
   parameterName=FlameEmissiveMask
   samplerSource=SSM_Wrap_WorldGroupSettings
   samplerType=SAMPLERTYPE_Masks
   texture=&AllGreen
  }
  n48 = AppendVector(A=Constant(), B=LinearInterpolate(A=Constant() {r=-1.1}, Alpha=@n47.B) {constB=1.1})
  n49 = TextureCoordinate()
  n50 = AppendVector(A=Constant() {r=3.1545}, B=Constant() {r=2.1709})
  n51 = TextureSampleParameter2D(UVs=Add(A=Panner(Coordinate=@n46, Speed=@n48) {speedX=1}, B=Multiply(A=@n49, B=@n50))) {
   group=patches_material
   parameterName=Mask Texture
   samplerSource=SSM_Wrap_WorldGroupSettings
   texture=/BRCosmetics/Athena/Effects/Particles/Skydiving/NS_Contrail_Chainmail/T_WispyFireNoise
  }
  NamedRerouteDeclaration(None=@n51.R) {name=UV Distortion2, nodeColor=<rgba 0.4531 1 0 1>}
  n52 = NamedRerouteUsage()
  n53 = TextureSampleParameter2D() {
   group=patches_material
   parameterName=Flame_mask
   samplerSource=SSM_Wrap_WorldGroupSettings
   samplerType=SAMPLERTYPE_Masks
   texture=&AllRed
  }
  n54 = LinearInterpolate(A=Constant() {r=0.009}, Alpha=@n53.B) {constB=0}
  n55 = TextureSampleParameter2D(UVs=Add(A=TextureCoordinate(), B=Multiply(A=@n52, B=@n54) {constB=0.01})) {
   group=patches_material
   parameterName=Flame_mask
   samplerSource=SSM_Wrap_WorldGroupSettings
   samplerType=SAMPLERTYPE_Masks
   texture=&AllRed
  }
  n56 = ScalarParameter() {defaultValue=1.7561, parameterName=Eye flame size}
  n57 = CurveAtlasRowParameter(CurveTime=Saturate(None=Multiply(A=@n55.R, B=@n56))) {
   atlas=&patch_ColorAtlas
   curve=&/CC_patch_eye
   defaultValue=1
   group=patches_material
   parameterName=Curve
  }
  NamedRerouteDeclaration(None=Multiply(A=@n57, B=ScalarParameter() {defaultValue=2.5672, parameterName=Eye brightness})) {name=Eye flame, nodeColor=<rgba 0.9219 0 1 1>}
  n58 = TextureCoordinate()
  n59 = Constant() {r=16}
  n60 = Constant()
  n61 = Constant() {r=1.2}
  n62 = TextureSampleParameter2D(UVs=Panner(Coordinate=Multiply(A=@n58, B=@n59), Speed=AppendVector(A=@n60, B=@n61)) {speedY=-0.5}) {
   group=patches_material
   parameterName=Distortion Texture
   samplerSource=SSM_Wrap_WorldGroupSettings
   samplerType=SAMPLERTYPE_Grayscale
   texture=&LowResBlurredNoise_Radish_LF_Root
  }
  NamedRerouteDeclaration(None=LinearInterpolate(A=Constant() {r=-0.3084}, Alpha=@n62.R) {constA=-0.2}) {name=EYE UV Distortion, nodeColor=<rgba 1 0.2109 0 1>}
 ]}
