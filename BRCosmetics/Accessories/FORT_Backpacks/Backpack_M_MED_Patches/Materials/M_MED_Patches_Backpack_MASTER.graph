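# Master material for the M_MED_Patches backpack cosmetic, built on material attributes with the shading model set by expression.
# The shared character PreFX function feeds an optional Generic Gem effect, chosen by static switch and driven by channels of the SkinFX_Mask texture.
# The result runs through the Patches flame function, then the character PostFX and CINE PostFX functions.
# A CharacterSkinFX call also reads PreFX output but feeds no other node shown.
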Material /BRCosmetics/Accessories/FORT_Backpacks/Backpack_M_MED_Patches/Materials/M_MED_Patches_Backpack_MASTER
graph {
 properties {
  bAutomaticallySetUsageInEditor=false
  bUseMaterialAttributes=true
  bUsedWithMeshDeformer=true
  bUsedWithMorphTargets=true
  bUsedWithSkeletalMesh=true
  pixelDepthOffsetMode=PDOM_Legacy
  refractionMethod=RM_IndexOfRefraction
  shadingModel=MSM_FromMaterialExpression
  subsurfaceProfile=&SS_HeroSkin_02
  usedShadingModels=Default Lit | Subsurface | Clear Coat | Subsurface Profile | Cloth | Eye
 }
 nodes [
  n1 = MaterialFunctionCall() {materialFunction=&BaseFN_CharacterFunction_PreFX}
  n2 = TextureSample(
   UVs=TextureCoordinate()
   Tex=NamedRerouteDeclaration(None=TextureObjectParameter() {group=Skin FX, parameterName=SkinFX_Mask}) {name=SkinFX_Mask, nodeColor=<rgba 0.9842 0 1 1>}
  )
  n3 = MaterialFunctionCall!(
   MASK=StaticComponentMaskParameter(None=@n2.RGBA) {defaultR=true, group=Skin FX, parameterName=Generic Gem Effect Channel}
   Color Change Sequence=StaticComponentMaskParameter(None=NamedRerouteDeclaration(None=@n2.RGBA) {name=SkinFX_Mask_Sampled, nodeColor=<rgba 0.7475 0 1 1>}) {defaultB=true, group=Skin FX, parameterName=Generic Gem Color Change Channel}
   Material=@n1.Result
  ) {materialFunction=/Game/Weapons/WeaponSkins/MaterialFunction/GenericGem}
  n4 = MaterialFunctionCall(
   AO=@n1.AO
   Distance Blend=@n1.AO
   Fresnel=@n1.AO
   Gmap=@n1.AO
   Pre Skinned Local Position=@n1.AO
   Material=MaterialFunctionCall!(In=StaticSwitchParameter(True=@n3.Result, False=@n1.Result) {group=Skin FX, parameterName=Use Generic Gem}) {materialFunction=&/MF_Patches_flame}.Result
  ) {materialFunction=&BaseFN_CharacterFunction_PostFX}
  MaterialFunctionCall(In=@n4.Result) {materialFunction=&CINE_MF_CharPostFX}
  MaterialFunctionCall(
   PreskinnedPosition=@n1.Pre skinned Local Position
   Gmap=@n1.Pre skinned Local Position
   MaterialAttributes=@n1.Pre skinned Local Position
   AO=@n1.Pre skinned Local Position
  ) {materialFunction=&CharacterSkinFX}
 ]
}
